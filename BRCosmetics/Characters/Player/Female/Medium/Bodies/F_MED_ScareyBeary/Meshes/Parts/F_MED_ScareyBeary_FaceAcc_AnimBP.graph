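# Post-copy Animation Blueprint for the ScareyBeary female face accessory, adding secondary motion to its ears and hood.
# It copies the attached parent's pose and curves, layering upper arms from the Body mesh cached when character customization loads.
# A Control Rig fed the parachute-open state precedes ear AnimDynamics, stacked additive hood bone offsets, then spring-driven hood dynamics.
# A second, bone-relative hood dynamics chain is left unconnected to the output.
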AnimBlueprint /BRCosmetics/Characters/Player/Female/Medium/Bodies/F_MED_ScareyBeary/Meshes/Parts/F_MED_ScareyBeary_FaceAcc_AnimBP
target_skeleton=&/F_MED_ScareyBeary_FaceAcc_Skeleton
graphs [
 AnimGraph {nodes [
   n1 = Anim Dynamics - Bone: dyn_hood_l(Alpha=0.5, ExternalForce="0,0,0") {
    angularDampingOverride=0.999
    angularSpringConstant=6000
    bAngularSpring=true
    bGravityOverrideInSimSpace=true
    bOverrideAngularDamping=true
    bOverrideLinearDamping=true
    bUseGravityOverride=true
    linearDampingOverride=1
    linearSpringConstant=50
    simulationSpace=BoneRelative
    alphaBoolBlend {blendOption=Linear}
    alphaScaleBiasClamp {bInterpResult=true}
    boundBone {boneName=dyn_hood_l}
    componentLinearAccScale=<0.2 0.2 0.2>
    componentLinearVelScale=<0.2 0.2 0.2>
    gravityOverride {z=-980}
    physicsBodyDefinitions [{
      boundBone {boneName=dyn_hood_l}
      boxExtents=<75 75 75>
      constraintSetup {angularLimitsMax=<5 15 5>, angularLimitsMin=<-5 -15 -5>, angularTarget=<0 0 0>, linearAxesMax=<0 0 1>, linearAxesMin=<0 0 -1>}
      localJointOffset=<0.1 0 0>
     }]
    relativeSpaceBone {boneName=pelvis}
   }
   n2 = Anim Dynamics - Bone: dyn_hood_bck(ComponentPose=@n1, Alpha=0.5, ExternalForce="0,0,0") {
    angularDampingOverride=0.999
    angularSpringConstant=6000
    bAngularSpring=true
    bGravityOverrideInSimSpace=true
    bOverrideAngularDamping=true
    bOverrideLinearDamping=true
    bUseGravityOverride=true
    linearDampingOverride=1
    linearSpringConstant=50
    simulationSpace=BoneRelative
    alphaBoolBlend {blendOption=Linear}
    alphaScaleBiasClamp {bInterpResult=true}
    boundBone {boneName=dyn_hood_bck}
    componentLinearAccScale=<0.2 0.2 0.2>
    componentLinearVelScale=<0.2 0.2 0.2>
    gravityOverride {z=-980}
    physicsBodyDefinitions [{
      boundBone {boneName=dyn_hood_bck}
      boxExtents=<75 75 75>
      constraintSetup {angularLimitsMax=<5 15 5>, angularLimitsMin=<-5 -15 -5>, angularTarget=<0 0 0>}
      localJointOffset=<0.1 0 0>
     }]
    relativeSpaceBone {boneName=pelvis}
   }
   Anim Dynamics - Bone: dyn_hood_r(ComponentPose=@n2, Alpha=0.5, ExternalForce="0,0,0") {
    angularDampingOverride=0.999
    angularSpringConstant=6000
    bAngularSpring=true
    bGravityOverrideInSimSpace=true
    bOverrideAngularDamping=true
    bOverrideLinearDamping=true
    bUseGravityOverride=true
    linearDampingOverride=1
    linearSpringConstant=50
    simulationSpace=BoneRelative
    alphaBoolBlend {blendOption=Linear}
    alphaScaleBiasClamp {bInterpResult=true}
    boundBone {boneName=dyn_hood_r}
    componentLinearAccScale=<0.2 0.2 0.2>
    componentLinearVelScale=<0.2 0.2 0.2>
    gravityOverride {z=980}
    physicsBodyDefinitions [{
      boundBone {boneName=dyn_hood_r}
      boxExtents=<75 75 75>
      constraintSetup {angularLimitsMax=<5 15 5>, angularLimitsMin=<-5 -15 -5>, angularTarget=<0 0 0>, linearAxesMax=<0 0 1>, linearAxesMin=<0 0 -1>}
      localJointOffset=<0.1 0 0>
     }]
    relativeSpaceBone {boneName=pelvis}
   }
   n3 = Get Body Mesh() -> (Body Mesh:Skeletal Mesh Component Object Reference)
   n4 = Control Rig!(
    Source=Layered blend per bone(
     BasePose=Copy Pose From Mesh() {bCopyCurves=true, bUseAttachedParent=true}
     BlendPoses_0=Copy Pose From Mesh(SourceMeshComponent=@n3) {bUseAttachedParent=true}
    ) {layerSetup [{branchFilters [{boneName=upperarm_l}, {boneName=upperarm_r}]}]}
    Alpha
    is_Parachute:bool=Get bIsParachuteOpened()
   ) -> (Pose) {
    alphaBoolBlend {blendOption=Linear}
    controlRigAssetReference {blueprintRigClass=&/F_MED_ScareyBeary_FaceAcc_CtrlRig.F_MED_ScareyBeary_FaceAcc_CtrlRig_C}
   }
   n5 = Anim Dynamics - Bone: dyn_ear_r(ComponentPose=Local To Component(LocalPose=@n4), Alpha=1, ExternalForce="0,0,0") {
    angularDampingOverride=0.9
    angularSpringConstant=3000
    bAngularSpring=true
    bGravityOverrideInSimSpace=true
    bOverrideAngularDamping=true
    bUseGravityOverride=true
    gravityScale=0.35
    simulationSpace=RootRelative
    alphaBoolBlend {blendOption=Linear}
    alphaScaleBiasClamp {scale=0.2}
    boundBone {boneName=dyn_ear_r}
    componentLinearAccScale=<1 1 1>
    componentLinearVelScale=<1 1 1>
    gravityOverride {z=-1000}
    physicsBodyDefinitions [{
      sphereCollisionRadius=0
      boundBone {boneName=dyn_ear_r}
      boxExtents=<10 20 20>
      constraintSetup {angularLimitsMax=<0 50 80>, angularLimitsMin=<0 -50 -50>, angularTarget=<1 0 0>}
      localJointOffset=<10 0 0>
     }]
   }
   n6 = Anim Dynamics - Bone: dyn_ear_l(ComponentPose=@n5, Alpha=1, ExternalForce="0,0,0") {
    angularDampingOverride=0.9
    angularSpringConstant=3000
    bAngularSpring=true
    bGravityOverrideInSimSpace=true
    bOverrideAngularDamping=true
    bUseGravityOverride=true
    gravityScale=0.35
    simulationSpace=RootRelative
    alphaBoolBlend {blendOption=Linear}
    alphaScaleBiasClamp {scale=0.2}
    boundBone {boneName=dyn_ear_l}
    componentLinearAccScale=<1 1 1>
    componentLinearVelScale=<1 1 1>
    gravityOverride {z=-1000}
    physicsBodyDefinitions [{
      sphereCollisionRadius=0
      boundBone {boneName=dyn_ear_l}
      boxExtents=<10 20 20>
      constraintSetup {angularLimitsMax=<0 50 80>, angularLimitsMin=<0 -50 -80>, angularTarget=<1 0 0>}
      localJointOffset=<10 0 0>
     }]
   }
   n7 = "Transform (Modify) Bone - Bone: dyn_hood_l"(ComponentPose=@n6, Alpha=1, Translation="0,0,4", Rotation="0.081,13.354,20.236", Scale="1,1,0.5") {
    rotationMode=BMM_Additive
    rotationSpace=BCS_BoneSpace
    scaleMode=BMM_Replace
    scaleSpace=BCS_BoneSpace
    translationMode=BMM_Additive
    translationSpace=BCS_BoneSpace
    alphaBoolBlend {blendOption=Linear}
    alphaScaleBiasClamp {bInterpResult=true, interpSpeedDecreasing=20, interpSpeedIncreasing=20}
    boneToModify {boneName=dyn_hood_l}
    rotation=<rot 0.081 13.354 20.236>
    scale {z=0.5}
    translation {z=4}
   }
   n8 = "Transform (Modify) Bone - Bone: dyn_hood_bck"(ComponentPose=@n7, Alpha=1, Translation="-1.026,0,5", Rotation="20.000,-0.000,-0.000", Scale="1.000,1.000,0.500") {
    rotationMode=BMM_Additive
    rotationSpace=BCS_BoneSpace
    scaleMode=BMM_Replace
    scaleSpace=BCS_BoneSpace
    translationMode=BMM_Additive
    translationSpace=BCS_BoneSpace
    alphaBoolBlend {blendOption=Linear}
    alphaScaleBiasClamp {bInterpResult=true, interpSpeedDecreasing=20, interpSpeedIncreasing=20}
    boneToModify {boneName=dyn_hood_bck}
    rotation {pitch=20}
    scale {z=0.5}
    translation {x=-1.026, z=5}
   }
   n9 = "Transform (Modify) Bone - Bone: dyn_hood_r"(ComponentPose=@n8, Alpha=1, Translation="0,0,3", Rotation="2.302,4.079,-19.501", Scale="1.000,1.000,0.500") {
    rotationMode=BMM_Additive
    rotationSpace=BCS_BoneSpace
    scaleMode=BMM_Replace
    scaleSpace=BCS_BoneSpace
    translationMode=BMM_Additive
    translationSpace=BCS_BoneSpace
    alphaBoolBlend {blendOption=Linear}
    alphaScaleBiasClamp {bInterpResult=true, interpSpeedDecreasing=20, interpSpeedIncreasing=20}
    boneToModify {boneName=dyn_hood_r}
    rotation=<rot 2.302 4.079 -19.501>
    scale {z=0.5}
    translation {z=3}
   }
   n10 = "Transform (Modify) Bone - Bone: dyn_hood_l"(ComponentPose=@n9, Alpha=1, Translation="0,0,3.5", Rotation="-4.985,-4.412,13.430", Scale="1,1,1") {
    rotationMode=BMM_Additive
    rotationSpace=BCS_BoneSpace
    scaleMode=BMM_Replace
    scaleSpace=BCS_BoneSpace
    translationMode=BMM_Additive
    translationSpace=BCS_BoneSpace
    alphaBoolBlend {blendOption=Linear}
    alphaScaleBiasClamp {bInterpResult=true, interpSpeedDecreasing=20, interpSpeedIncreasing=20}
    boneToModify {boneName=dyn_hood_l}
    rotation=<rot -4.985 -4.412 13.43>
    translation {z=3.5}
   }
   n11 = "Transform (Modify) Bone - Bone: dyn_hood_bck"(ComponentPose=@n10, Alpha=1, Translation="0.13,-1.858,2", Rotation="10.000,-0.043,0.003", Scale="1,1,1") {
    rotationMode=BMM_Additive
    rotationSpace=BCS_BoneSpace
    scaleMode=BMM_Replace
    scaleSpace=BCS_BoneSpace
    translationMode=BMM_Additive
    translationSpace=BCS_BoneSpace
    alphaBoolBlend {blendOption=Linear}
    alphaScaleBiasClamp {bInterpResult=true, interpSpeedDecreasing=20, interpSpeedIncreasing=20}
    boneToModify {boneName=dyn_hood_bck}
    rotation=<rot 9.9999 -0.043 0.003>
    translation=<0.13 -1.858 2>
   }
   n12 = "Transform (Modify) Bone - Bone: dyn_hood_r"(ComponentPose=@n11, Alpha=1, Translation="-0.067,-0.082,2", Rotation="-9.977,-0.062,-5.044", Scale="1,1,1") {
    rotationMode=BMM_Additive
    rotationSpace=BCS_BoneSpace
    scaleMode=BMM_Replace
    scaleSpace=BCS_BoneSpace
    translationMode=BMM_Additive
    translationSpace=BCS_BoneSpace
    alphaBoolBlend {blendOption=Linear}
    alphaScaleBiasClamp {bInterpResult=true, interpSpeedDecreasing=20, interpSpeedIncreasing=20}
    boneToModify {boneName=dyn_hood_r}
    rotation=<rot -9.977 -0.062 -5.044>
    translation=<-0.067 -0.082 2>
   }
   n13 = Anim Dynamics - Bone: dyn_hood_l(ComponentPose=@n12, Alpha=0.8, ExternalForce="0,0,0") {
    angularDampingOverride=0.9
    angularSpringConstant=3000
    bAngularSpring=true
    bGravityOverrideInSimSpace=true
    bOverrideAngularDamping=true
    bUseGravityOverride=true
    gravityScale=0.35
    simulationSpace=RootRelative
    alphaBoolBlend {blendOption=Linear}
    alphaScaleBiasClamp {scale=0.2}
    boundBone {boneName=dyn_hood_l}
    componentLinearAccScale=<1 1 1>
    componentLinearVelScale=<1 1 1>
    gravityOverride {z=-1000}
    physicsBodyDefinitions [{
      sphereCollisionRadius=0
      boundBone {boneName=dyn_hood_l}
      boxExtents=<10 20 20>
      constraintSetup {angularLimitsMax=<0 50 20>, angularLimitsMin=<0 -50 -80>, angularTarget=<1 0 0>}
      localJointOffset=<10 0 0>
     }]
   }
   n14 = Anim Dynamics - Bone: dyn_hood_r(ComponentPose=@n13, Alpha=0.8, ExternalForce="0,0,0") {
    angularDampingOverride=0.9
    angularSpringConstant=3000
    bAngularSpring=true
    bGravityOverrideInSimSpace=true
    bOverrideAngularDamping=true
    bUseGravityOverride=true
    gravityScale=0.35
    simulationSpace=RootRelative
    alphaBoolBlend {blendOption=Linear}
    alphaScaleBiasClamp {scale=0.2}
    boundBone {boneName=dyn_hood_r}
    componentLinearAccScale=<1 1 1>
    componentLinearVelScale=<1 1 1>
    gravityOverride {z=-980}
    physicsBodyDefinitions [{
      sphereCollisionRadius=0
      boundBone {boneName=dyn_hood_r}
      boxExtents=<10 20 20>
      constraintSetup {angularLimitsMax=<0 50 80>, angularLimitsMin=<0 -50 -20>, angularTarget=<1 0 0>}
      localJointOffset=<10 0 0>
     }]
   }
   n15 = Anim Dynamics - Bone: dyn_hood_bck(ComponentPose=@n14, Alpha=0.8, ExternalForce="0,0,0") {
    angularDampingOverride=0.9
    angularSpringConstant=3000
    bAngularSpring=true
    bGravityOverrideInSimSpace=true
    bOverrideAngularDamping=true
    bUseGravityOverride=true
    gravityScale=0.35
    simulationSpace=RootRelative
    alphaBoolBlend {blendOption=Linear}
    alphaScaleBiasClamp {scale=0.2}
    boundBone {boneName=dyn_hood_bck}
    componentLinearAccScale=<1 1 1>
    componentLinearVelScale=<1 1 1>
    gravityOverride {z=-980}
    physicsBodyDefinitions [{
      sphereCollisionRadius=0
      boundBone {boneName=dyn_hood_bck}
      boxExtents=<10 20 20>
      constraintSetup {angularLimitsMax=<0 50 20>, angularLimitsMin=<0 -50 -20>, angularTarget=<1 0 0>}
      localJointOffset=<10 0 0>
     }]
   }
   Output Pose(Result=Component To Local(ComponentPose=@n15)) {layerGroup=None}
   Reroute Node(InputPin:Component Space Pose Link Structure=Reroute Node()) -> (OutputPin:Component Space Pose Link Structure)
  ]}
 EventGraph {nodes [
   n1 = Event OnCharacterCustomizationLoaded()
   > Set Body Mesh(Body Mesh:Skeletal Mesh Component Object Reference=Get Array Item with EFort Custom Part Type(Array=@n1.CustomMeshParts)) -> (Output_Get:Skeletal Mesh Component Object Reference)
  ]}
]
variables [Body Mesh {type=SkeletalMeshComponent}]
